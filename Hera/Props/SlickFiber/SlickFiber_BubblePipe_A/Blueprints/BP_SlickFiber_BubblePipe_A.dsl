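; Blueprint for the SlickFiber Bubble Pipe A prop, which silences its own audio when it dies.
; On BeginPlay it binds a Pipe Audio Stops handler to the OnDeathPlayEffects delegate.
; The handler gathers every AudioComponent on the actor and deactivates each, ignoring the damage details it receives.
; The construction script only calls the parent.
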(event EventBeginPlay
  (EventDispatchers|BindEventtoOnDeathPlayEffectsDelegate (delegate Custom|Pipe%20Audio%20Stops)))

(event Custom|Pipe%20Audio%20Stops
  (params (Damage :type Float%20%28single-precision%29) (DamageTags :type Gameplay%20Tag%20Container%20Structure%20%28by%20ref%29) (Momentum :type Vector) (HitInfo :type Hit%20Result%20Structure%20%28by%20ref%29) (InstigatedBy :type Fort%20Pawn%20Object%20Reference) (DamageCauser :type Actor%20Object%20Reference) (EffectContext :type Gameplay%20Effect%20Context%20Handle%20Structure))
  (for _array_element (Actor|GetComponentsByClass "/Script/Engine.AudioComponent")
    (Components|Activation|Deactivate _array_element)))

(fn ConstructionScript
  (|CallParentFunction))
