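; Blueprint for an audio-only water spline on a car wash bubble pipe, which places audio components along a spline path.
; The construction script loops over an index range sized from the spline's point count plus one, adding an audio component per index.
; Each component moves to the world location of the spline point at that index.
; BeginPlay is empty, so the script adds no runtime logic.
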(event EventBeginPlay)

(fn ConstructionScript
  (for _index (range (+ (Spline|GetNumberOfSplinePoints (Variables|Default|GetSpline)) 1))
    (bind _returnvalue (AddComponent|NODE_AddAudioComponent-0))
    (Transformation|SetWorldLocation _returnvalue (Spline|GetLocationatSplinePoint (Variables|Default|GetSpline) _index "World") false false)))
